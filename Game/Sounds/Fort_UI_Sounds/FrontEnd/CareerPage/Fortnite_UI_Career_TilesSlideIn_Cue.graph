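# Sound cue for the front-end Career page UI, named for the sound of tiles sliding in.
# A Mixer layers two medium whoosh waves at equal level, one from the generic UI toolkit and one from the Battle Pass toolkit.
# Each wave passes through its own Modulator with fixed reduced volume, and the cue further lowers overall volume and pitch.
# The cue is assigned to the RenderedCinematics sound class rather than a UI-named one.
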SoundCue /Game/Sounds/Fort_UI_Sounds/FrontEnd/CareerPage/Fortnite_UI_Career_TilesSlideIn_Cue
root=Mixer_0
nodes [
 {id=Mixer_0, children [Modulator_1,Modulator_0], properties {inputVolume [1,1]}}
 {id=Modulator_0, children [WavePlayer_0], properties {pitchMax=1, volumeMax=0.3, volumeMin=0.3}}
 {id=Modulator_1, children [WavePlayer_1], properties {pitchMax=1, volumeMax=0.3, volumeMin=0.3}}
 {
  id=WavePlayer_0
  properties {soundWaveAssetPtr=/Game/Sounds/Fort_UI_Sounds/FrontEnd/_GenericToolkit_/Whooshes/Misc_Whooshes/wav/UI_Toolkit_Whoosh_Med_06}
 }
 {id=WavePlayer_1, properties {soundWaveAssetPtr=/Game/Sounds/Fort_UI_Sounds/FrontEnd/BattlePass/BP_ToolKit/BP_Modal_Whoosh_Med_02}}
]
properties {maxDistance=2097152, pitchMultiplier=0.8, soundClassObject=/Game/Sounds/FortSoundClassesAndModes/RenderedCinematics, volumeMultiplier=0.05}
