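# SoundCue for the generic button sound on the front-end Career page, inferred from its name and folder.
# A single wave player feeds the root directly and plays the UI_FrontEnd_Career_Button_Generic_01 wave, with no randomization, modulation or mixing.
# It routes through the Fort_UI sound class at reduced volume.
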SoundCue /Game/Sounds/Fort_UI_Sounds/FrontEnd/CareerPage/Fortnite_UI_Career_Button_Generic_01_Cue
root=WavePlayer_2
nodes [{id=WavePlayer_2, properties {soundWaveAssetPtr=&/UI_FrontEnd_Career_Button_Generic_01}}]
properties {maxDistance=2097152, soundClassObject=&Fort_UI, volumeMultiplier=0.5}
