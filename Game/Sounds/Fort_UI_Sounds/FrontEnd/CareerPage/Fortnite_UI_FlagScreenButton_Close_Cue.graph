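# Sound cue for closing the flag screen button on the front-end Career page UI.
# A single wave player plays the career flag close sound wave directly, with no randomization or mixing nodes.
# The cue lowers both volume and pitch and is assigned to the RenderedCinematics sound class.
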SoundCue /Game/Sounds/Fort_UI_Sounds/FrontEnd/CareerPage/Fortnite_UI_FlagScreenButton_Close_Cue
root=WavePlayer_0
nodes [{id=WavePlayer_0, properties {soundWaveAssetPtr=&/UI_FrontEnd_Career_Flag_Close}}]
properties {maxDistance=2097152, pitchMultiplier=0.8, soundClassObject=/Game/Sounds/FortSoundClassesAndModes/RenderedCinematics, volumeMultiplier=0.2}
