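# Sound cue for the flag screen button on the front-end Career page, in the Fort_UI sound class.
# A Mixer layers the standard front-end button click with a flag modal appear sound.
# A Modulator pins the appear sound to a low, fixed volume, which the Mixer boosts slightly.
# A concurrency override caps simultaneous instances and sets a short retrigger time.
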SoundCue /Game/Sounds/Fort_UI_Sounds/FrontEnd/CareerPage/Fortnite_UI_FlagScreenButton_Cue
root=Mixer_0
nodes [
 {id=Mixer_0, children [WavePlayer_2,Modulator_4], properties {inputVolume [1,1.25]}}
 {id=Modulator_4, children [WavePlayer_0], properties {volumeMax=0.2, volumeMin=0.2}}
 {id=WavePlayer_0, properties {soundWaveAssetPtr=&/UI_FrontEnd_Career_FlagModalAppear}}
 {id=WavePlayer_2, properties {soundWaveAssetPtr=/Game/Sounds/Fort_UI_Sounds/FrontEnd/New_FrontEnd/UI_FN_Button_Click_Main_01}}
]
properties {
 bOverrideConcurrency=true
 maxDistance=2097152
 soundClassObject=&Fort_UI
 volumeMultiplier=1
 concurrencyOverrides {maxCount=3, retriggerTime=0.15}
}
